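; Coastal standing lantern building prop with an animated flicker and tintable lantern and rod meshes.
; Each tick sums two weighted absolute sine waves of game time and clamps the result to a minimum brightness.
; That scalar feeds custom primitive data slot 0 for the lantern material and scales an optional assigned point light, keeping both synced.
; The construction script tints both meshes from building texture data, defaulting to grey.
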(event EventBeginPlay)

(event Collision|EventActorBeginOverlap
  (params (OtherActor :type Actor%20Object%20Reference)))

(event EventTick
  (params (DeltaSeconds :type Float%20%28single-precision%29))
  (bind _returnvalue (Utilities|Time|GetGameTimeinSeconds))
  (bind _flickerspeed1 (Variables|Flicker|GetFlickerSpeed1))
  (bind _returnvalue_2 (* _returnvalue _flickerspeed1))
  (bind _returnvalue_3 (Math|Trig|Sin(Radians) _returnvalue_2))
  (bind _returnvalue_4 (Math|Float|Absolute(Float) _returnvalue_3))
  (bind _flickerspeed2 (Variables|Flicker|GetFlickerSpeed2))
  (bind _returnvalue_5 (* _returnvalue _flickerspeed2))
  (bind _returnvalue_6 (Math|Trig|Sin(Radians) _returnvalue_5))
  (bind _returnvalue_7 (Math|Float|Absolute(Float) _returnvalue_6))
  (bind _minbrightness (Variables|Flicker|GetMinBrightness))
  (bind _flickeramplitude2 (Variables|Flicker|GetFlickerAmplitude2))
  (bind _returnvalue_8 (* _returnvalue_7 _flickeramplitude2))
  (bind _flickeramplitude1 (Variables|Flicker|GetFlickerAmplitude1))
  (bind _returnvalue_9 (* _returnvalue_4 _flickeramplitude1))
  (bind _returnvalue_10 (+ _returnvalue_9 _returnvalue_8))
  (bind _returnvalue_11 (Math|Float|Clamp(Float) _returnvalue_10 _minbrightness))
  (bind _flickerpointlight (Variables|Flicker|GetFlickerPointLight))
  ;[0.85 0.55 0.15] Write flicker to CPD slot 0 on the lantern mesh
  ; M_LanternFlicker_CPD reads this via PerInstanceCustomData(index=0) in the shader
  (Rendering|Material|SetCustomPrimitiveDataFloat (Variables|Default|GetSM_LightSource_CoastalStandingLantern_A) 0 _returnvalue_11)
  ;[0.15 0.85 0.35] Drive point light with the same scalar - guaranteed in sync
  ; Assign FlickerPointLight in Details panel per placed instance
  (if (Utilities|IsValidClass (Utilities|GetClass _flickerpointlight))
    (Rendering|Components|Light|SetIntensity _flickerpointlight (* (Variables|Flicker|GetBaseLightIntensity) _returnvalue_11))))

(fn ConstructionScript
  (bind _l_tdindex (LocalVariables|GetL_TDIndex))
  (bind _sm_lightsource_coastalstandinglantern_a (Variables|Default|GetSM_LightSource_CoastalStandingLantern_A))
  (bind _sm_lightsource_coastalstandinglanternrod_a (Variables|Default|GetSM_LightSource_CoastalStandingLanternRod_A))
  (|CallParentFunction)
  (bind _outtexturedataarray (Building|GetBuildingTextureData))
  (if (Utilities|Array|IsValidIndex _outtexturedataarray _l_tdindex)
    (Default|SetColoronSecondaryMesh (Math|Conversions|ToLinearColor(Color) (Variables|/Script/FortniteGame.BuildingTextureData|GetTintColor (Utilities|Array|Get(acopy) _outtexturedataarray _l_tdindex))) _sm_lightsource_coastalstandinglantern_a)
    (Default|SetColoronSecondaryMesh (Math|Conversions|ToLinearColor(Color) (Variables|/Script/FortniteGame.BuildingTextureData|GetTintColor (Utilities|Array|Get(acopy) _outtexturedataarray _l_tdindex))) _sm_lightsource_coastalstandinglanternrod_a)
    (else
      (Default|SetColoronSecondaryMesh "(R=0.500000,G=0.500000,B=0.500000,A=1.000000)" _sm_lightsource_coastalstandinglantern_a)
      (Default|SetColoronSecondaryMesh "(R=0.500000,G=0.500000,B=0.500000,A=1.000000)" _sm_lightsource_coastalstandinglanternrod_a))))
